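# Niagara system for the idle glow on the necklace of the female medium FishBowl character body cosmetic.
# Necklace_Light bursts two persistent particles that follow the user-supplied skeletal mesh and drive light, component and layered sprite renderers.
# Light, air-glow and decal-glow colors derive from a shared System.LightColor, with brightness scaled by time of day.
# A second emitter, Round_Flare, adds mesh-attached flare sprites but is disabled.
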NiagaraSystem /BRCosmetics/Characters/Player/Female/Medium/Bodies/F_MED_FishBowl/FX/NS_FishBowl_Body_Necklace_Idle
dependencies {
 usedDataInterfaces [&NiagaraDataInterfaceCurve]
 usedDynamicInputs []
 usedModules [
  &self:SystemSpawnScript.SystemScriptSource.SystemScriptGraph.NiagaraNodeAssignment_0.SetVariables_#1
  &SystemState
 ]
}
emitters {
 Necklace_Light {
  emitterUpdateScript [
   &EmitterState(
    Inactive Response=::NewEnumerator1
    Life Cycle Mode=::NewEnumerator1
    Loop Behavior=::NewEnumerator1
    Loop Duration Mode=::NewEnumerator1
    UseLoopDelay=false
   )
   &SpawnBurst_Instantaneous(Spawn Count=2, Use Spawn Probability=false, UseLoopCountLimit=false)
  ]
  particleSpawnScript [
   &PartitionParticles(Particles Per Partition=1, Partition Mode:ENiagara_PartitionMode=::NewEnumerator1)
   &InitializeParticle(
    Color Mode=::NewEnumerator0
    Position Offset=<50 0 0>
    Sprite Size Mode=::NewEnumerator3
    Uniform Sprite Size=1
    UsePositionOffset=false
    Write Material Random=true
   ) {module=InitializeParticle001}
   /SetVariables(
    Particles.LightColor=&MultiplyLinearColor(
     Linear Color=$System.LightColor
     Scale Factor=&MakeVector4_FromFloat(Float=&FNTimeOfDayScalar(Night Scale=1, Morning Scale=2, Day Scale=2, Evening Scale=2))
    )
    Particles.LightComponentColor:LinearColor=&MultiplyLinearColor(Linear Color=$System.LightColor, Scale Factor=&MakeVector4_FromFloat(Float=1))
    Particles.LightComponentIntensity:NiagaraFloat=&FNTimeOfDayScalar(Night Scale=1.5, Morning Scale=2.5, Day Scale=2.5, Evening Scale=2.5)
    Particles.LightExponent:NiagaraFloat=&FNTimeOfDayScalar(Night Scale=4, Morning Scale=4, Day Scale=4, Evening Scale=4)
    Particles.LightRadius=25
   ) {isSetParameters=true}
   /SetVariables(
    Particles.AirGlowColor:LinearColor=&MultiplyLinearColor(
     Linear Color=$System.LightColor
     Scale Factor=&MakeVector4_FromFloat(Float=&FNTimeOfDayScalar(Night Scale=0.4, Morning Scale=0.5, Day Scale=0.5, Evening Scale=0.5))
    )
    Particles.AirGlowColorLow:LinearColor=&MultiplyLinearColor(
     Linear Color=$System.LightColor
     Scale Factor=&MakeVector4_FromFloat(Float=&FNTimeOfDayScalar(Night Scale=0.15, Morning Scale=0.25, Day Scale=0.25, Evening Scale=0.25))
    )
    Particles.AirGlowSpriteScale:Vector2f=&Multiply_Vector2D_ByFloat(Float=10, Vector2D=<1 1>)
   ) {isSetParameters=true}
   /SetVariables(
    Particles.DecalGlowColor:LinearColor=&MultiplyLinearColor(
     Linear Color=&MultiplyLinearColor(
      Linear Color=$System.LightColor
      Scale Factor=&MakeVector4_FromFloat(Float=&FNTimeOfDayScalar(Night Scale=0.5, Morning Scale=1, Evening Scale=1))
     )
     Scale Factor=&MakeVector4_FromFloat(Float=&FN_FrontEndFloatModifier(Gameplay=1, Front End=1))
    )
    Particles.DecalGlowSpriteScale:Vector2f=&Multiply_Vector2D_ByFloat(Float=20, Vector2D=<1 1>)
   ) {isSetParameters=true}
   &FN_Sprite_Attributes(
    Color=<rgba 0.3235 0.1059 1 0.07>
    Create Color Attribute=true
    Create Sprite Rotation Attribute=true
    Create SpriteSize Attribute=true
    SpriteRotation=-30
    SpriteSize=&Vector2DFromFloat(Value=9)
    Write to Module Namespace=true
   )
  ]
  particleUpdateScript [
   &ParticleState(Kill Particles When Lifetime Has Elapsed=false, Let Infinitely Lived Particles Die When Emitter Deactivates=true)
   &SkeletalMeshLocation(
    Bone Sampling Mode=::NewEnumerator4
    MaskSpawnGroup=false
    Sampled Position Offset=<2 0 0>
    Skeletal Mesh=$User.SkeletalMesh
    Use Sampled Mesh Attribute as Ribbon ID=false
   ) {module=SkeletalMeshLocation001}
   &CameraOffset(Camera Offset Amount=2)
  ]
  rendererClasses [
   &NiagaraLightRendererProperties
   &NiagaraSpriteRendererProperties
   &NiagaraComponentRendererProperties
  ]
  renderers [
   {rendererClass=&NiagaraLightRendererProperties, rendererIndex=0}
   {rendererClass=&NiagaraSpriteRendererProperties, rendererIndex=1}
   {rendererClass=&NiagaraSpriteRendererProperties, rendererIndex=2}
   {rendererClass=&NiagaraComponentRendererProperties, rendererIndex=3}
   {rendererClass=&NiagaraSpriteRendererProperties, rendererIndex=4}
   {rendererClass=&NiagaraSpriteRendererProperties, rendererIndex=5}
   {rendererClass=&NiagaraSpriteRendererProperties, rendererIndex=6}
  ]
 }
 Round_Flare {
  bEnabled=false
  emitterUpdateScript [
   &EmitterState(
    Inactive Response=::NewEnumerator1
    Life Cycle Mode=::NewEnumerator1
    Loop Behavior=::NewEnumerator1
    Loop Duration Mode=::NewEnumerator1
    UseLoopDelay=false
   )
   &SpawnBurst_Instantaneous(Use Spawn Probability=false, UseLoopCountLimit=false)
  ]
  particleSpawnScript [
   &ApplyInitialForces()
   &InitializeParticle(
    Color=&MultiplyLinearColor(Linear Color=$System.LightColor, Scale Factor=&MakeVector4_FromFloat(Float=0.25))
    Lifetime=1.1
    Position Offset=<0 15 20>
    Sprite Size Mode=::NewEnumerator3
    Uniform Sprite Size=50
    UsePositionOffset=true
    Write Material Random=true
   )
   &CameraOffset(Camera Offset Amount=3)
   &AlignSpriteToMeshOrientation(
    Mesh Orientation Relative Sprite Alignment Vector=<1 0 0>
    Mesh Orientation Relative Sprite Facing Vector:Vector3f=<0 1 1>
    SetAlignment=false
    SetFacing=true
   )
   /SetVariables(Particles.FlareCamOffset=2, Particles.FlareColor=<rgba 0.1879 0.0757 0.8 0.5>) {isSetParameters=true}
  ]
  particleUpdateScript [
   &ParticleState(Kill Particles When Lifetime Has Elapsed=false, Let Infinitely Lived Particles Die When Emitter Deactivates=true)
   &SkeletalMeshLocation(
    Bone Sampling Mode=::NewEnumerator4
    MaskSpawnGroup=false
    Sampled Position Offset=<2 0 1>
    Skeletal Mesh=$User.SkeletalMesh
    Use Sampled Mesh Attribute as Ribbon ID=false
   ) {module=SkeletalMeshLocation001}
   &DynamicMaterialParameters(
    Index 0 Param 1=0.5
    Index 0 Param 2=0.5
    Index 0 Param 3=1
    Index 0 Param 4=1
    Module.Param0WriteEnabled=true
    Module.Param1WriteEnabled=true
    Module.Param2WriteEnabled=true
    Module.Param3WriteEnabled=true
   )
   &ScaleSpriteSize(
    Scale Sprite Size Mode=::NewEnumerator0
    Uniform Scale Factor=&RandomRangeFloat(Minimum=0.5, Maximum=1)
   )
   /SetVariables(Particles.Wide Flare Size:Vector2f=&Multiply_Vector2D_ByFloat(
     Float=&RandomRangeFloat(Minimum=0.75, Maximum=1.25)
     Vector2D=&MakeVector2D(X=&FNTimeOfDayScalar(Night Scale=50, Morning Scale=60, Day Scale=60, Evening Scale=60), Y=5)
    )) {isSetParameters=true}
  ]
  rendererClasses [&NiagaraSpriteRendererProperties]
  renderers [
   {rendererClass=&NiagaraSpriteRendererProperties}
   {rendererClass=&NiagaraSpriteRendererProperties, rendererIndex=1}
  ]
 }
}
summary {
 systemName=NS_FishBowl_Body_Necklace_Idle
 emitters [
  {
   emitterName=Necklace_Light
   rendererClasses [
    &NiagaraLightRendererProperties
    &NiagaraSpriteRendererProperties
    &NiagaraComponentRendererProperties
   ]
  }
  {bEnabled=false, emitterName=Round_Flare, rendererClasses [&NiagaraSpriteRendererProperties]}
 ]
 userVariables [
  User.bIsFrontEnd {type=NiagaraBool}
  User.bIsFrontEndPreview {type=NiagaraBool}
  User.SkeletalMesh {
   type=NiagaraDataInterfaceSkeletalMesh
   defaultValue {
    struct=NiagaraExt_VariableValue_DataInterface
    value {dataInterface=&self:NiagaraDataInterfaceSkeletalMesh_0}
   }
  }
 ]
}
