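# Physics asset for the F_MED_FishBowl female player body, covering its dangling belt and cloth bones plus the hips and upper legs.
# Capsules on the pelvis, both thighs and the right thigh twist bone form the body volumes around the dyn_ bones.
# A three-link belt chain and two-link cloth chain each hang from a master bone with free swing and twist on the pelvis.
# Links within each chain have angular limits, with tighter swing on the cloth and tighter twist on the belt.
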PhysicsAsset /BRCosmetics/Characters/Player/Female/Medium/Bodies/F_MED_FishBowl/Meshes/F_MED_FishBowl_Physics
bodies {
 dyn_belt_01 [{shapeType=Capsule, length=3, radius=1, shapeName=dyn_belt_01_capsule, center=<2.5 -0.5 0>, rotation=<rot 90 -89.4453 -79.4452>}]
 dyn_belt_02 [{shapeType=Capsule, length=5, radius=1.005, shapeName=dyn_belt_03_capsule, center=<3.2 -0.5 0>, rotation=<rot 90 0 0>}]
 dyn_belt_03 [{shapeType=Capsule, length=6.5, radius=1.005, shapeName=dyn_belt_03_capsule, center=<4 -0.5 0>, rotation=<rot 90 0 0>}]
 dyn_belt_MSTR [{shapeType=Capsule, length=0.505, radius=0.505, shapeName=dyn_belt_MSTR_capsule, rotation=<rot 0 -90 0>}]
 dyn_cloth_01 [{shapeType=Capsule, length=2.755, radius=4.255, shapeName=dyn_cloth_01_capsule, center=<7 -3 0>, rotation=<rot 5.0112 -91.3304 7.4229>}]
 dyn_cloth_02 [{shapeType=Capsule, length=6.6615, radius=4.7893, shapeName=dyn_cloth_02_capsule, center=<6.5 -3.5 0>}]
 dyn_cloth_MSTR [{shapeType=Capsule, length=0.505, radius=0.505, shapeName=dyn_cloth_MSTR_capsule, rotation=<rot 0 -90 0>}]
 pelvis [
  {shapeType=Capsule, length=9.0207, radius=11.159, center=<-2.4234 -0.7117 0>}
  {shapeType=Capsule, length=5.4375, radius=2.1875, shapeName=pelvis_capsule, center=<1.1029 8.1309 2.8282>, rotation=<rot 70 -0.0003 10>}
 ]
 thigh_l [{shapeType=Capsule, length=27.6999, radius=8.0778, center=<-22.189 -0.2313 -2.0084>, rotation=<rot -0.2732 90.0487 -86.8754>}]
 thigh_r [{shapeType=Capsule, length=27.6999, radius=8.0778, center=<22.1994 0.2304 1.5397>, rotation=<rot -0.2732 90.0487 -86.8754>}]
 thigh_twist_01_r [{shapeType=Capsule, length=27.0749, radius=8.7028, center=<-9.6221 0.5044 -0.1988>, rotation=<rot -0.2732 90.0487 -86.8754>}]
}
constraints [
 {bone1Name=thigh_l, bone2Name=pelvis, swing1LimitDegrees=45, swing2LimitDegrees=45, twistLimitDegrees=45}
 {bone1Name=thigh_r, bone2Name=pelvis, swing1LimitDegrees=45, swing2LimitDegrees=45, twistLimitDegrees=45}
 {
  bone1Name=dyn_belt_MSTR
  bone2Name=pelvis
  swing1LimitDegrees=45
  swing1Motion=Free
  swing2LimitDegrees=45
  swing2Motion=Free
  twistLimitDegrees=45
  twistMotion=Free
 }
 {bone1Name=dyn_belt_01, bone2Name=dyn_belt_MSTR, swing1LimitDegrees=45, swing2LimitDegrees=45, twistLimitDegrees=15}
 {bone1Name=dyn_belt_02, bone2Name=dyn_belt_01, swing1LimitDegrees=45, swing2LimitDegrees=45, twistLimitDegrees=15}
 {bone1Name=dyn_belt_03, bone2Name=dyn_belt_02, swing1LimitDegrees=45, swing2LimitDegrees=45, twistLimitDegrees=15}
 {
  bone1Name=dyn_cloth_MSTR
  bone2Name=pelvis
  swing1LimitDegrees=45
  swing1Motion=Free
  swing2LimitDegrees=45
  swing2Motion=Free
  twistLimitDegrees=45
  twistMotion=Free
 }
 {bone1Name=dyn_cloth_01, bone2Name=dyn_cloth_MSTR, swing1LimitDegrees=35, swing2LimitDegrees=35, twistLimitDegrees=25}
 {bone1Name=dyn_cloth_02, bone2Name=dyn_cloth_01, swing1LimitDegrees=35, swing2LimitDegrees=35, twistLimitDegrees=25}
 {
  bone1Name=thigh_twist_01_r
  bone2Name=thigh_r
  swing1LimitDegrees=45
  swing1Motion=Free
  swing2LimitDegrees=45
  swing2Motion=Free
  twistLimitDegrees=45
  twistMotion=Free
 }
]
