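# Physics asset for the jacket of the female medium FishBowl player cosmetic, set up for two dangling tassels.
# Left and right four-capsule chains hang from side master bodies, which join a central master constrained to spine_04.
# Each chain's first link is tightly limited, while lower links get wider swing and moderate twist.
# Pelvis, spine and thigh capsules, plus paired spheres on spine_04, approximate the body around the tassels.
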PhysicsAsset /BRCosmetics/Characters/Player/Female/Medium/Bodies/F_MED_FishBowl/Meshes/F_MED_FishBowl_jacket_Physics
bodies {
 dyn_tassle_MSTR [{shapeType=Capsule, length=0.505, radius=0.505, shapeName=dyn_tassle_MSTR_capsule, rotation=<rot 0 -90 0>}]
 dyn_tassle_l_01 [{shapeType=Capsule, length=0.63, radius=0.88, shapeName=dyn_tassle_r_01_capsule, center=<1.5 0 0>, rotation=<rot 90 0 0>}]
 dyn_tassle_l_02 [{shapeType=Capsule, length=2.63, radius=0.88, shapeName=dyn_tassle_l_02_capsule, center=<1.5 0 0>, rotation=<rot 90 0 0>}]
 dyn_tassle_l_03 [{shapeType=Capsule, length=4.63, radius=0.88, shapeName=dyn_tassle_l_03_capsule, center=<3 0 0>, rotation=<rot 90 0 0>}]
 dyn_tassle_l_04 [{shapeType=Capsule, length=6.13, radius=0.88, shapeName=dyn_tassle_l_04_capsule, center=<3.5 0 0>, rotation=<rot 90 0 0>}]
 dyn_tassle_l_MSTR [{shapeType=Capsule, length=0.505, radius=0.505, shapeName=dyn_tassle_l_MSTR_capsule, rotation=<rot 0 -90 0>}]
 dyn_tassle_r_01 [{shapeType=Capsule, length=0.63, radius=0.88, shapeName=dyn_tassle_r_01_capsule, center=<1.5 0 0>, rotation=<rot 90 0 0>}]
 dyn_tassle_r_02 [{shapeType=Capsule, length=2.13, radius=0.88, shapeName=dyn_tassle_r_02_capsule, center=<1.5 0 0>, rotation=<rot 90 0 0>}]
 dyn_tassle_r_03 [{shapeType=Capsule, length=3.63, radius=0.88, shapeName=dyn_tassle_r_03_capsule, center=<2.5 0 0>, rotation=<rot 90 0 0>}]
 dyn_tassle_r_04 [{shapeType=Capsule, length=4.38, radius=0.88, shapeName=dyn_tassle_r_04_capsule, center=<2.5 0 0>, rotation=<rot 90 0 0>}]
 dyn_tassle_r_MSTR [{shapeType=Capsule, length=0.505, radius=0.505, shapeName=dyn_tassle_r_MSTR_capsule, rotation=<rot 0 -90 0>}]
 pelvis [
  {shapeType=Capsule, length=9.0207, radius=11.159, center=<-2.4234 -0.7117 0>}
  {shapeType=Capsule, length=5.1875, radius=2.4375, shapeName=pelvis_capsule, center=<1.1029 8.1309 2.8282>, rotation=<rot 70 -0.0003 10>}
 ]
 spine_02 [{shapeType=Capsule, length=5.7205, radius=8.505, center=<-0.2527 2.0894 -0.0001>}]
 spine_03 [{shapeType=Capsule, length=5.3789, radius=9.3507, center=<0 2.1962 -0.0001>}]
 spine_04 [
  {shapeType=Sphere, radius=7.4, shapeName=spine_04_sphere, center=<-0.5 5.4 5>}
  {shapeType=Sphere, radius=7.4, shapeName=spine_04_sphere, center=<-0.5 5.4 -5>}
  {shapeType=Capsule, length=9.1388, radius=8.5305, center=<-1.061 3.7553 0>}
 ]
 thigh_l [{shapeType=Capsule, length=27.6999, radius=8.0778, center=<-22.189 -0.2313 -2.0084>, rotation=<rot -0.2732 90.0487 -86.8754>}]
 thigh_r [{shapeType=Capsule, length=27.6999, radius=8.0778, center=<22.1994 0.2304 1.5397>, rotation=<rot -0.2732 90.0487 -86.8754>}]
}
constraints [
 {bone1Name=spine_04, bone2Name=spine_03, swing1LimitDegrees=45, swing2LimitDegrees=45, twistLimitDegrees=45}
 {bone1Name=thigh_l, bone2Name=pelvis, swing1LimitDegrees=45, swing2LimitDegrees=45, twistLimitDegrees=45}
 {bone1Name=thigh_r, bone2Name=pelvis, swing1LimitDegrees=45, swing2LimitDegrees=45, twistLimitDegrees=45}
 {
  bone1Name=spine_03
  bone2Name=spine_02
  swing1LimitDegrees=45
  swing1Motion=Free
  swing2LimitDegrees=45
  swing2Motion=Free
  twistLimitDegrees=45
  twistMotion=Free
 }
 {
  bone1Name=spine_02
  bone2Name=pelvis
  swing1LimitDegrees=45
  swing1Motion=Free
  swing2LimitDegrees=45
  swing2Motion=Free
  twistLimitDegrees=45
  twistMotion=Free
 }
 {bone1Name=dyn_tassle_r_04, bone2Name=dyn_tassle_r_03, swing1LimitDegrees=45, swing2LimitDegrees=45, twistLimitDegrees=15}
 {
  bone1Name=dyn_tassle_MSTR
  bone2Name=spine_04
  swing1LimitDegrees=45
  swing1Motion=Free
  swing2LimitDegrees=45
  swing2Motion=Free
  twistLimitDegrees=45
  twistMotion=Free
 }
 {
  bone1Name=dyn_tassle_l_MSTR
  bone2Name=dyn_tassle_MSTR
  swing1LimitDegrees=45
  swing1Motion=Free
  swing2LimitDegrees=45
  swing2Motion=Free
  twistLimitDegrees=45
  twistMotion=Free
 }
 {bone1Name=dyn_tassle_l_01, bone2Name=dyn_tassle_l_MSTR, swing1LimitDegrees=5, swing2LimitDegrees=5, twistLimitDegrees=3}
 {bone1Name=dyn_tassle_l_02, bone2Name=dyn_tassle_l_01, swing1LimitDegrees=45, swing2LimitDegrees=45, twistLimitDegrees=15}
 {bone1Name=dyn_tassle_l_03, bone2Name=dyn_tassle_l_02, swing1LimitDegrees=45, swing2LimitDegrees=45, twistLimitDegrees=15}
 {bone1Name=dyn_tassle_l_04, bone2Name=dyn_tassle_l_03, swing1LimitDegrees=45, swing2LimitDegrees=45, twistLimitDegrees=15}
 {
  bone1Name=dyn_tassle_r_MSTR
  bone2Name=dyn_tassle_MSTR
  swing1LimitDegrees=45
  swing1Motion=Free
  swing2LimitDegrees=45
  swing2Motion=Free
  twistLimitDegrees=45
  twistMotion=Free
 }
 {bone1Name=dyn_tassle_r_01, bone2Name=dyn_tassle_r_MSTR, swing1LimitDegrees=5, swing2LimitDegrees=5, twistLimitDegrees=3}
 {bone1Name=dyn_tassle_r_02, bone2Name=dyn_tassle_r_01, swing1LimitDegrees=45, swing2LimitDegrees=45, twistLimitDegrees=15}
 {bone1Name=dyn_tassle_r_03, bone2Name=dyn_tassle_r_02, swing1LimitDegrees=45, swing2LimitDegrees=45, twistLimitDegrees=15}
]
