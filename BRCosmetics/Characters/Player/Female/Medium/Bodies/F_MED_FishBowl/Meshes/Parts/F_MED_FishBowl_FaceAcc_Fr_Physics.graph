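# Physics asset for the FishBowl female medium character's FaceAcc part, setting up secondary motion for right-front hair and strand bone chains.
# Both chains hang from the head through small master bodies, alongside capsules on head, neck, upper spine, right clavicle and upper arm that give them body shapes to collide with.
# Body and master joints swing and twist freely, while hair links have moderate, uneven swing limits and strand links stay nearly rigid.
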PhysicsAsset /BRCosmetics/Characters/Player/Female/Medium/Bodies/F_MED_FishBowl/Meshes/Parts/F_MED_FishBowl_FaceAcc_Fr_Physics
bodies {
 clavicle_r [{shapeType=Capsule, length=2.1669, radius=8.5644, center=<-6.5041 -0.1497 2.1395>, rotation=<rot 85.606 0 0>}]
 dyn_hair_r_Front_00 [{shapeType=Capsule, length=1.875, radius=3.125, shapeName=dyn_hair_r_Front_01_capsule, center=<3 0 0>, rotation=<rot 90 0 0>}]
 dyn_hair_r_Front_01 [{shapeType=Capsule, length=2.0625, radius=2.9375, shapeName=dyn_hair_r_Front_02_capsule, center=<2.5 2 0>, rotation=<rot 90 83.4 38.3998>}]
 dyn_hair_r_Front_02 [{
   shapeType=Capsule
   length=2.3125
   radius=2.4375
   shapeName=dyn_hair_r_Front_03_capsule
   center=<3.5 2 0>
   rotation=<rot 90 -116.0664 -131.0666>
  }]
 dyn_hair_r_Front_03 [{shapeType=Capsule, length=4.875, radius=1.875, shapeName=dyn_hair_r_Front_04_capsule, center=<3.5 1 0>, rotation=<rot 90 0 0>}]
 dyn_hair_r_Front_MSTR [{shapeType=Capsule, length=0.505, radius=0.505, shapeName=dyn_hair_r_Front_MSTR_capsule, rotation=<rot 0 -90 0>}]
 dyn_strands_01 [{shapeType=Capsule, length=3, radius=1, shapeName=dyn_strands_01_capsule, center=<2 0 0>, rotation=<rot 90 0 0>}]
 dyn_strands_02 [{shapeType=Capsule, length=3, radius=1, shapeName=dyn_strands_02_capsule, center=<2 0 0>, rotation=<rot 90 0 0>}]
 dyn_strands_03 [{shapeType=Capsule, length=3, radius=1, shapeName=dyn_strands_03_capsule, center=<2 0 0>, rotation=<rot 90 0 0>}]
 dyn_strands_04 [{shapeType=Capsule, length=3, radius=1, shapeName=dyn_strands_04_capsule, center=<2 0 0>, rotation=<rot 90 0 0>}]
 dyn_strands_05 [{shapeType=Capsule, length=3, radius=1, shapeName=dyn_strands_05_capsule, center=<3 0 0>, rotation=<rot 90 0 0>}]
 dyn_strands_MSTR [{shapeType=Capsule, length=0.505, radius=0.505, shapeName=dyn_strands_MSTR_capsule, rotation=<rot 90 0 0>}]
 head [
  {shapeType=Capsule, length=3.0796, radius=6.7593, shapeName=head_capsule, center=<2.3025 0.855 -0.5075>, rotation=<rot 90 0 5.0001>}
  {shapeType=Capsule, length=5.375, radius=2, shapeName=head_capsule, center=<1.3844 6.7235 0>, rotation=<rot -90 90 -90.0001>}
 ]
 neck_01 [{shapeType=Capsule, length=5.005, radius=5.5675, shapeName=neck_01_capsule, rotation=<rot 0 -90 0>}]
 spine_04 [{shapeType=Capsule, length=4.1566, radius=10.5126, shapeName=spine_04_capsule, center=<0.306 0.5327 0>, rotation=<rot 0 -92.6847 0>}]
 spine_05 [{shapeType=Capsule, length=9.7816, radius=10.5126, shapeName=spine_04_capsule, center=<0.306 0.5327 0>, rotation=<rot 0 -92.6847 0>}]
 upperarm_r [{shapeType=Capsule, length=13.4229, radius=4.1429, shapeName=upperarm_r_capsule, center=<-5 1.25 0.0001>, rotation=<rot 0 -90 90>}]
}
constraints [
 {
  bone1Name=head
  bone2Name=neck_01
  swing1LimitDegrees=45
  swing1Motion=Free
  swing2LimitDegrees=45
  swing2Motion=Free
  twistLimitDegrees=45
  twistMotion=Free
 }
 {
  bone1Name=spine_05
  bone2Name=spine_04
  swing1LimitDegrees=45
  swing1Motion=Free
  swing2LimitDegrees=45
  swing2Motion=Free
  twistLimitDegrees=45
  twistMotion=Free
 }
 {
  bone1Name=clavicle_r
  bone2Name=spine_05
  swing1LimitDegrees=45
  swing1Motion=Free
  swing2LimitDegrees=45
  swing2Motion=Free
  twistLimitDegrees=45
  twistMotion=Free
 }
 {
  bone1Name=upperarm_r
  bone2Name=clavicle_r
  swing1LimitDegrees=45
  swing1Motion=Free
  swing2LimitDegrees=45
  swing2Motion=Free
  twistLimitDegrees=45
  twistMotion=Free
 }
 {
  bone1Name=neck_01
  bone2Name=spine_05
  swing1LimitDegrees=45
  swing1Motion=Free
  swing2LimitDegrees=45
  swing2Motion=Free
  twistLimitDegrees=45
  twistMotion=Free
 }
 {bone1Name=dyn_strands_05, bone2Name=dyn_strands_04, swing1LimitDegrees=1, swing2LimitDegrees=1, twistLimitDegrees=1}
 {
  bone1Name=dyn_strands_MSTR
  bone2Name=head
  swing1LimitDegrees=45
  swing1Motion=Free
  swing2LimitDegrees=45
  swing2Motion=Free
  twistLimitDegrees=45
  twistMotion=Free
 }
 {bone1Name=dyn_strands_01, bone2Name=dyn_strands_MSTR, swing1LimitDegrees=3, swing2LimitDegrees=3, twistLimitDegrees=3}
 {bone1Name=dyn_strands_02, bone2Name=dyn_strands_01, swing1LimitDegrees=3, swing2LimitDegrees=3, twistLimitDegrees=3}
 {bone1Name=dyn_strands_03, bone2Name=dyn_strands_02, swing1LimitDegrees=3, swing2LimitDegrees=3, twistLimitDegrees=3}
 {bone1Name=dyn_strands_04, bone2Name=dyn_strands_03, swing1LimitDegrees=3, swing2LimitDegrees=3, twistLimitDegrees=3}
 {bone1Name=dyn_hair_r_Front_03, bone2Name=dyn_hair_r_Front_02, swing1LimitDegrees=35, swing2LimitDegrees=15, twistLimitDegrees=5}
 {
  bone1Name=dyn_hair_r_Front_MSTR
  bone2Name=head
  swing1LimitDegrees=45
  swing1Motion=Free
  swing2LimitDegrees=45
  swing2Motion=Free
  twistLimitDegrees=45
  twistMotion=Free
 }
 {bone1Name=dyn_hair_r_Front_00, bone2Name=dyn_hair_r_Front_MSTR, swing1LimitDegrees=10, swing2LimitDegrees=10, twistLimitDegrees=5}
 {bone1Name=dyn_hair_r_Front_01, bone2Name=dyn_hair_r_Front_00, swing1LimitDegrees=35, swing2LimitDegrees=15, twistLimitDegrees=5}
 {bone1Name=dyn_hair_r_Front_02, bone2Name=dyn_hair_r_Front_01, swing1LimitDegrees=35, swing2LimitDegrees=15, twistLimitDegrees=5}
]
