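# Physics asset for the FishBowl female medium character's face accessory part, simulating its dynamic hair bones against upper-body shapes.
# Capsules cover the head, neck, upper spine, clavicles and upper arms.
# Five hair chains (back, left, right, two sides) hang from head-attached master bodies through angularly free joints, then use swing and twist limits.
# Extra angularly free constraints cross-link each side chain to the same-side left or right chain.
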PhysicsAsset /BRCosmetics/Characters/Player/Female/Medium/Bodies/F_MED_FishBowl/Meshes/Parts/F_MED_FishBowl_FaceAcc_Physics
bodies {
 clavicle_l [{shapeType=Capsule, length=1.9169, radius=8.8144, center=<6.4709 0.3011 -2.1395>, rotation=<rot 85.606 0 0>}]
 clavicle_r [{shapeType=Capsule, length=2.1669, radius=8.5644, center=<-6.5041 0.3959 2.1395>, rotation=<rot 85.606 0 0>}]
 dyn_hair_bk_00 [{shapeType=Capsule, length=3.3125, radius=4.4375, shapeName=dyn_Hair_Back_03_capsule, center=<5 -2 0>, rotation=<rot 90 0 0>}]
 dyn_hair_bk_01 [{shapeType=Capsule, length=2.5, radius=5.25, shapeName=dyn_Hair_Back_04_capsule, center=<5 -2 0>, rotation=<rot 90 0 0>}]
 dyn_hair_bk_02 [{shapeType=Capsule, length=3.1875, radius=4.5625, shapeName=dyn_Hair_Back_05_capsule, center=<5 -2 0>, rotation=<rot 90 0 0>}]
 dyn_hair_bk_MSTR [{shapeType=Capsule, length=0.505, radius=0.505, shapeName=dyn_Hair_Back_MSTR_capsule}]
 dyn_hair_l_00 [{shapeType=Capsule, length=0.3125, radius=5.3125, shapeName=dyn_Hair_BackL_03_capsule, center=<5 -2 0>, rotation=<rot 90 0 0>}]
 dyn_hair_l_01 [{shapeType=Capsule, length=0.375, radius=5.25, shapeName=dyn_Hair_BackL_04_capsule, center=<5 -2 0>, rotation=<rot 90 0 0>}]
 dyn_hair_l_02 [{shapeType=Capsule, length=1, radius=4.625, shapeName=dyn_Hair_BackL_05_capsule, center=<5 -2 0>, rotation=<rot 90 0 0>}]
 dyn_hair_l_MSTR [{shapeType=Capsule, length=0.505, radius=0.505, shapeName=dyn_Hair_BackL_MSTR_capsule}]
 dyn_hair_r_00 [{shapeType=Capsule, length=1.3125, radius=4.3125, shapeName=dyn_Hair_BackR_03_capsule, center=<5 -2 0>, rotation=<rot 90 0 0>}]
 dyn_hair_r_01 [{shapeType=Capsule, length=0.9375, radius=4.6875, shapeName=dyn_Hair_BackR_04_capsule, center=<5 -2 0>, rotation=<rot 90 0 0>}]
 dyn_hair_r_02 [{shapeType=Capsule, length=0.5, radius=5.625, shapeName=dyn_Hair_BackR_05_capsule, center=<5 -2 0>, rotation=<rot 90 0 0>}]
 dyn_hair_r_MSTR [{shapeType=Capsule, length=0.505, radius=0.505, shapeName=dyn_Hair_BackR_MSTR_capsule}]
 dyn_hair_side_l_01 [{shapeType=Capsule, length=1.5625, radius=5.5, shapeName=dyn_Hair_BackL_03_capsule, center=<4 0 2>, rotation=<rot 90 0 0>}]
 dyn_hair_side_l_02 [{shapeType=Capsule, length=5, radius=3, shapeName=dyn_Hair_BackL_03_capsule, center=<4 1 0>, rotation=<rot 90 0 0>}]
 dyn_hair_side_l_03 [{shapeType=Capsule, length=2.6875, radius=4.0625, shapeName=dyn_Hair_BackL_04_capsule, center=<4 -2 0>, rotation=<rot 90 0 0>}]
 dyn_hair_side_l_04 [{shapeType=Capsule, length=1.6875, radius=3.9375, shapeName=dyn_Hair_BackL_05_capsule, center=<5 -2 0>, rotation=<rot 90 0 0>}]
 dyn_hair_side_l_MSTR [{shapeType=Capsule, length=0.505, radius=0.505, shapeName=dyn_Hair_BackL_MSTR_capsule}]
 dyn_hair_side_r_01 [{shapeType=Capsule, length=2.0625, radius=3.4375, shapeName=dyn_Hair_BackL_03_capsule, center=<5 0 -2>, rotation=<rot 90 0 0>}]
 dyn_hair_side_r_02 [{shapeType=Capsule, length=5, radius=3, shapeName=dyn_Hair_BackL_03_capsule, center=<4 1 0>, rotation=<rot 90 0 0>}]
 dyn_hair_side_r_03 [{shapeType=Capsule, length=1.875, radius=4.875, shapeName=dyn_Hair_BackL_04_capsule, center=<4 -2 0>, rotation=<rot 90 0 0>}]
 dyn_hair_side_r_04 [{shapeType=Capsule, length=0.625, radius=5, shapeName=dyn_Hair_BackL_05_capsule, center=<5 -2 0>, rotation=<rot 90 0 0>}]
 dyn_hair_side_r_MSTR [{shapeType=Capsule, length=0.505, radius=0.505, shapeName=dyn_Hair_BackR_MSTR_capsule}]
 head [
  {shapeType=Capsule, length=3.0796, radius=6.7593, shapeName=head_capsule, center=<2.3025 0.855 -0.5075>, rotation=<rot 90 0 5.0001>}
  {shapeType=Capsule, length=5.375, radius=2, shapeName=head_capsule, center=<1.3844 6.7235 0>, rotation=<rot -90 90 -90.0001>}
 ]
 neck_01 [{shapeType=Capsule, length=4.255, radius=5.5675, shapeName=neck_01_capsule, rotation=<rot 0 -90 0>}]
 spine_04 [{shapeType=Capsule, length=4.5941, radius=10.0751, shapeName=spine_04_capsule, center=<2.4959 -1.4941 0>, rotation=<rot 0 -92.6847 0>}]
 spine_05 [{shapeType=Capsule, length=10.8161, radius=10.2701, shapeName=spine_05_capsule, center=<1.2872 -0.7071 -0.1551>}]
 upperarm_l [{shapeType=Capsule, length=18.0812, radius=4.1347, center=<10.8056 -0.4936 0.3456>, rotation=<rot 90 0 0>}]
 upperarm_r [{shapeType=Capsule, length=18.0812, radius=4.1347, center=<-10.8056 0.4936 -0.3456>, rotation=<rot 90 0 0>}]
}
constraints [
 {
  bone1Name=head
  bone2Name=neck_01
  swing1LimitDegrees=45
  swing1Motion=Free
  swing2LimitDegrees=45
  swing2Motion=Free
  twistLimitDegrees=45
  twistMotion=Free
 }
 {
  bone1Name=spine_05
  bone2Name=spine_04
  swing1LimitDegrees=45
  swing1Motion=Free
  swing2LimitDegrees=45
  swing2Motion=Free
  twistLimitDegrees=45
  twistMotion=Free
 }
 {
  bone1Name=clavicle_l
  bone2Name=spine_05
  swing1LimitDegrees=45
  swing1Motion=Free
  swing2LimitDegrees=45
  swing2Motion=Free
  twistLimitDegrees=45
  twistMotion=Free
 }
 {
  bone1Name=upperarm_l
  bone2Name=clavicle_l
  swing1LimitDegrees=45
  swing1Motion=Free
  swing2LimitDegrees=45
  swing2Motion=Free
  twistLimitDegrees=45
  twistMotion=Free
 }
 {
  bone1Name=clavicle_r
  bone2Name=spine_05
  swing1LimitDegrees=45
  swing1Motion=Free
  swing2LimitDegrees=45
  swing2Motion=Free
  twistLimitDegrees=45
  twistMotion=Free
 }
 {
  bone1Name=upperarm_r
  bone2Name=clavicle_r
  swing1LimitDegrees=45
  swing1Motion=Free
  swing2LimitDegrees=45
  swing2Motion=Free
  twistLimitDegrees=45
  twistMotion=Free
 }
 {
  bone1Name=dyn_hair_bk_MSTR
  bone2Name=head
  swing1LimitDegrees=45
  swing1Motion=Free
  swing2LimitDegrees=45
  swing2Motion=Free
  twistLimitDegrees=45
  twistMotion=Free
 }
 {bone1Name=dyn_hair_bk_00, bone2Name=dyn_hair_bk_MSTR, swing1LimitDegrees=50, swing2LimitDegrees=15, twistLimitDegrees=10}
 {bone1Name=dyn_hair_bk_01, bone2Name=dyn_hair_bk_00, swing1LimitDegrees=50, swing2LimitDegrees=15, twistLimitDegrees=10}
 {bone1Name=dyn_hair_bk_02, bone2Name=dyn_hair_bk_01, swing1LimitDegrees=50, swing2LimitDegrees=15, twistLimitDegrees=10}
 {
  bone1Name=dyn_hair_l_MSTR
  bone2Name=head
  swing1LimitDegrees=45
  swing1Motion=Free
  swing2LimitDegrees=45
  swing2Motion=Free
  twistLimitDegrees=45
  twistMotion=Free
 }
 {bone1Name=dyn_hair_l_00, bone2Name=dyn_hair_l_MSTR, swing1LimitDegrees=50, swing2LimitDegrees=15, twistLimitDegrees=10}
 {bone1Name=dyn_hair_l_01, bone2Name=dyn_hair_l_00, swing1LimitDegrees=50, swing2LimitDegrees=15, twistLimitDegrees=10}
 {bone1Name=dyn_hair_l_02, bone2Name=dyn_hair_l_01, swing1LimitDegrees=50, swing2LimitDegrees=15, twistLimitDegrees=10}
 {
  bone1Name=dyn_hair_r_MSTR
  bone2Name=head
  swing1LimitDegrees=45
  swing1Motion=Free
  swing2LimitDegrees=45
  swing2Motion=Free
  twistLimitDegrees=45
  twistMotion=Free
 }
 {bone1Name=dyn_hair_r_00, bone2Name=dyn_hair_r_MSTR, swing1LimitDegrees=50, swing2LimitDegrees=10, twistLimitDegrees=10}
 {bone1Name=dyn_hair_r_01, bone2Name=dyn_hair_r_00, swing1LimitDegrees=50, swing2LimitDegrees=10, twistLimitDegrees=10}
 {bone1Name=dyn_hair_r_02, bone2Name=dyn_hair_r_01, swing1LimitDegrees=50, swing2LimitDegrees=10, twistLimitDegrees=10}
 {
  bone1Name=dyn_hair_side_l_MSTR
  bone2Name=head
  swing1LimitDegrees=45
  swing1Motion=Free
  swing2LimitDegrees=45
  swing2Motion=Free
  twistLimitDegrees=45
  twistMotion=Free
 }
 {bone1Name=dyn_hair_side_l_01, bone2Name=dyn_hair_side_l_MSTR, swing1LimitDegrees=50, swing2LimitDegrees=15, twistLimitDegrees=10}
 {bone1Name=dyn_hair_side_l_02, bone2Name=dyn_hair_side_l_01, swing1LimitDegrees=50, swing2LimitDegrees=15, twistLimitDegrees=10}
 {bone1Name=dyn_hair_side_l_03, bone2Name=dyn_hair_side_l_02, swing1LimitDegrees=50, swing2LimitDegrees=15, twistLimitDegrees=10}
 {bone1Name=dyn_hair_side_l_04, bone2Name=dyn_hair_side_l_03, swing1LimitDegrees=50, swing2LimitDegrees=15, twistLimitDegrees=10}
 {
  bone1Name=dyn_hair_side_r_MSTR
  bone2Name=head
  swing1LimitDegrees=45
  swing1Motion=Free
  swing2LimitDegrees=45
  swing2Motion=Free
  twistLimitDegrees=45
  twistMotion=Free
 }
 {bone1Name=dyn_hair_side_r_01, bone2Name=dyn_hair_side_r_MSTR, swing1LimitDegrees=50, swing2LimitDegrees=10, twistLimitDegrees=10}
 {bone1Name=dyn_hair_side_r_02, bone2Name=dyn_hair_side_r_01, swing1LimitDegrees=50, swing2LimitDegrees=10, twistLimitDegrees=10}
 {bone1Name=dyn_hair_side_r_03, bone2Name=dyn_hair_side_r_02, swing1LimitDegrees=50, swing2LimitDegrees=10, twistLimitDegrees=10}
 {bone1Name=dyn_hair_side_r_04, bone2Name=dyn_hair_side_r_03, swing1LimitDegrees=50, swing2LimitDegrees=10, twistLimitDegrees=10}
 {
  bone1Name=neck_01
  bone2Name=spine_05
  swing1LimitDegrees=45
  swing1Motion=Free
  swing2LimitDegrees=45
  swing2Motion=Free
  twistLimitDegrees=45
  twistMotion=Free
 }
 {
  bone1Name=dyn_hair_side_r_03
  bone2Name=dyn_hair_r_01
  swing1LimitDegrees=45
  swing1Motion=Free
  swing2LimitDegrees=45
  swing2Motion=Free
  twistLimitDegrees=45
  twistMotion=Free
 }
 {
  bone1Name=dyn_hair_side_l_03
  bone2Name=dyn_hair_l_01
  swing1LimitDegrees=45
  swing1Motion=Free
  swing2LimitDegrees=45
  swing2Motion=Free
  twistLimitDegrees=45
  twistMotion=Free
 }
]
