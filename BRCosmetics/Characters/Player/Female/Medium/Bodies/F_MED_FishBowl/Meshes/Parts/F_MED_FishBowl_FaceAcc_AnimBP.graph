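# AnimBlueprint for the FishBowl female FaceAcc part, animating its hair bones on top of the parent mesh's copied pose.
# Weighted constraints, a crouch-driven Control Rig and additive jump, skydive and parachute offsets shape the pose.
# Two RigidBody sims and AnimDynamics follow, with blended constraints replacing motion on many bones when rigid bodies are disabled.
# EventGraph sets state-based gravity and pose inputs, damping inertia and jump offsets when backbling is detected.
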AnimBlueprint /BRCosmetics/Characters/Player/Female/Medium/Bodies/F_MED_FishBowl/Meshes/Parts/F_MED_FishBowl_FaceAcc_AnimBP
target_skeleton=&/F_MED_FishBowl_FaceAcc_Skeleton
graphs [
 AnimGraph {nodes [
   Constraint!(ComponentPose, Alpha=1) -> (Pose) {alphaBoolBlend {blendOption=Linear}, boneToModify=~}
   n1 = NOT Boolean(A=Get bRigidBodyEnabled())
   "Constraint - dyn_hair_l_00 (PP)"!(ComponentPose, bAlphaBoolEnabled:bool=@n1, ConstraintWeights_0=0.6, ConstraintWeights_1:float=0.4) -> (Pose) {
    alphaInputType=Bool
    alphaBoolBlend {blendInTime=0.1, blendOption=Linear, blendOutTime=0.1}
    boneToModify {boneName=dyn_hair_l_00}
    constraintSetup [{targetBone {boneName=spine_05}}, {targetBone {boneName=head}}]
    constraintWeights [1,1]
   }
   "Constraint - dyn_hair_bk_00 (PP)"!(ComponentPose, bAlphaBoolEnabled:bool=@n1, ConstraintWeights_0=0.4, ConstraintWeights_1:float=0.6) -> (Pose) {
    alphaInputType=Bool
    alphaBoolBlend {blendInTime=0.1, blendOption=Linear, blendOutTime=0.1}
    boneToModify {boneName=dyn_hair_bk_00}
    constraintSetup [{targetBone {boneName=head}}, {targetBone {boneName=spine_05}}]
    constraintWeights [1,1]
   }
   "Constraint - dyn_hair_r_00 (PP)"!(ComponentPose, bAlphaBoolEnabled:bool=@n1, ConstraintWeights_0=0.6, ConstraintWeights_1:float=0.4) -> (Pose) {
    alphaInputType=Bool
    alphaBoolBlend {blendInTime=0.1, blendOption=Linear, blendOutTime=0.1}
    boneToModify {boneName=dyn_hair_r_00}
    constraintSetup [{targetBone {boneName=spine_05}}, {targetBone {boneName=head}}]
    constraintWeights [1,1]
   }
   n2 = "Constraint - dyn_hair_l_MSTR (PP)"!(
    ComponentPose=Local To Component(LocalPose=Copy Pose From Mesh() {bUseAttachedParent=true})
    Alpha=1
    ConstraintWeights_0=0.8
    ConstraintWeights_1:float=0.2
   ) -> (Pose) {
    alphaBoolBlend {blendOption=Linear}
    boneToModify {boneName=dyn_hair_l_MSTR}
    constraintSetup [{targetBone {boneName=spine_05}}, {targetBone {boneName=head}}]
    constraintWeights [1,1]
   }
   n3 = "Constraint - dyn_hair_r_MSTR (PP)"!(ComponentPose=@n2, Alpha=1, ConstraintWeights_0=0.8, ConstraintWeights_1:float=0.2) -> (Pose) {
    alphaBoolBlend {blendOption=Linear}
    boneToModify {boneName=dyn_hair_r_MSTR}
    constraintSetup [{targetBone {boneName=spine_05}}, {targetBone {boneName=head}}]
    constraintWeights [1,1]
   }
   n4 = "Constraint - dyn_hair_bk_MSTR (PP)"!(ComponentPose=@n3, Alpha=1, ConstraintWeights_0=0.8, ConstraintWeights_1:float=0.2) -> (Pose) {
    alphaBoolBlend {blendOption=Linear}
    boneToModify {boneName=dyn_hair_bk_MSTR}
    constraintSetup [{targetBone {boneName=spine_05}}, {targetBone {boneName=head}}]
    constraintWeights [1,1]
   }
   n5 = "Constraint - dyn_hair_side_r_MSTR (PP)"!(ComponentPose=@n4, Alpha=1, ConstraintWeights_0=0.4, ConstraintWeights_1:float=0.6) -> (Pose) {
    alphaBoolBlend {blendOption=Linear}
    boneToModify {boneName=dyn_hair_side_r_MSTR}
    constraintSetup [{targetBone {boneName=clavicle_r}}, {targetBone {boneName=head}}]
    constraintWeights [1,1]
   }
   n6 = "Constraint - dyn_hair_side_l_MSTR (PP)"!(ComponentPose=@n5, Alpha=1, ConstraintWeights_0=0.4, ConstraintWeights_1:float=0.6) -> (Pose) {
    alphaBoolBlend {blendOption=Linear}
    boneToModify {boneName=dyn_hair_side_l_MSTR}
    constraintSetup [{targetBone {boneName=clavicle_l}}, {targetBone {boneName=head}}]
    constraintWeights [1,1]
   }
   n7 = "Constraint - dyn_hair_r_Front_MSTR (PP)"!(ComponentPose=@n6, Alpha=1, ConstraintWeights_0=0.6, ConstraintWeights_1:float=0.4) -> (Pose) {
    alphaBoolBlend {blendOption=Linear}
    boneToModify {boneName=dyn_hair_r_Front_MSTR}
    constraintSetup [{targetBone {boneName=clavicle_r}}, {targetBone {boneName=head}}]
    constraintWeights [1,1]
   }
   n8 = Control Rig!(
    Source=Component To Local(ComponentPose=@n7)
    Alpha
    crouch move:double=Get crouchmove() -> (crouchmove:double)
    crouch:double=Get crouch() -> (crouch:double)
   ) -> (Pose) {
    alphaBoolBlend {blendOption=Linear}
    controlRigAssetReference {blueprintRigClass=&/F_MED_FishBowl_FaceAcc_CtrlRig.F_MED_FishBowl_FaceAcc_CtrlRig_C}
   }
   n9 = "Constraint - dyn_strands_05 (PPPPPPPPP)"!(
    ComponentPose=Local To Component(LocalPose=@n8)
    Alpha=1
    ConstraintWeights_0=0.4759
    ConstraintWeights_1:float=0.0006
    ConstraintWeights_2:float=0.0003
    ConstraintWeights_3:float=0.4057
    ConstraintWeights_4:float=0.0645
    ConstraintWeights_5:float=0.0214
    ConstraintWeights_6:float=0.0032
    ConstraintWeights_7:float=0.0121
    ConstraintWeights_8:float=0.0163
   ) -> (Pose) {
    alphaBoolBlend {blendOption=Linear}
    boneToModify {boneName=dyn_strands_05}
    constraintSetup [
     {targetBone {boneName=head}}
     {targetBone {boneName=neck_02}}
     {targetBone {boneName=clavicle_r}}
     {targetBone {boneName=dyn_hair_side_r_01}}
     {targetBone {boneName=dyn_hair_side_r_02}}
     {targetBone {boneName=dyn_hair_side_r_03}}
     {targetBone {boneName=dyn_hair_side_r_04}}
     {targetBone {boneName=dyn_hair_r_00}}
     {targetBone {boneName=dyn_hair_r_01}}
    ]
    constraintWeights [1,1,1,1,1,1,1,1,1]
   }
   n10 = Get jumping() -> (jumping:double)
   n11 = "Transform (Modify) Bone - Bone: dyn_hair_bk_MSTR"(ComponentPose=@n9, Alpha=@n10, Translation="0,0,0", Rotation="0,-15,0", Scale="1,1,1") {
    rotationMode=BMM_Additive
    rotationSpace=BCS_BoneSpace
    alphaBoolBlend {blendOption=Linear}
    alphaScaleBiasClamp {bInterpResult=true, interpSpeedIncreasing=2}
    boneToModify {boneName=dyn_hair_bk_MSTR}
    rotation {yaw=-15}
   }
   n12 = "Transform (Modify) Bone - Bone: dyn_hair_r_MSTR"(ComponentPose=@n11, Alpha=@n10, Translation="0,0,0", Rotation="0,-15,0", Scale="1,1,1") {
    rotationMode=BMM_Additive
    rotationSpace=BCS_BoneSpace
    alphaBoolBlend {blendOption=Linear}
    alphaScaleBiasClamp {bInterpResult=true, interpSpeedIncreasing=2}
    boneToModify {boneName=dyn_hair_r_MSTR}
    rotation {yaw=-15}
   }
   n13 = "Transform (Modify) Bone - Bone: dyn_hair_l_MSTR"(ComponentPose=@n12, Alpha=@n10, Translation="0,0,0", Rotation="0,-15,0", Scale="1,1,1") {
    rotationMode=BMM_Additive
    rotationSpace=BCS_BoneSpace
    alphaBoolBlend {blendOption=Linear}
    alphaScaleBiasClamp {bInterpResult=true, interpSpeedIncreasing=2}
    boneToModify {boneName=dyn_hair_l_MSTR}
    rotation {yaw=-15}
   }
   n14 = "Transform (Modify) Bone - Bone: dyn_hair_side_r_MSTR"(ComponentPose=@n13, Alpha=@n10, Translation="0,0,0", Rotation="0,-7,0", Scale="1,1,1") {
    rotationMode=BMM_Additive
    rotationSpace=BCS_BoneSpace
    alphaBoolBlend {blendOption=Linear}
    alphaScaleBiasClamp {bInterpResult=true, interpSpeedIncreasing=2}
    boneToModify {boneName=dyn_hair_side_r_MSTR}
    rotation {yaw=-7}
   }
   n15 = "Transform (Modify) Bone - Bone: dyn_hair_side_l_MSTR"(ComponentPose=@n14, Alpha=@n10, Translation="0,0,0", Rotation="0,-7,0", Scale="1,1,1") {
    rotationMode=BMM_Additive
    rotationSpace=BCS_BoneSpace
    alphaBoolBlend {blendOption=Linear}
    alphaScaleBiasClamp {bInterpResult=true, interpSpeedIncreasing=2}
    boneToModify {boneName=dyn_hair_side_l_MSTR}
    rotation {yaw=-7}
   }
   n16 = Get bIsSkyDiving()
   n17 = "Transform (Modify) Bone - Bone: dyn_hair_bk_MSTR"!(ComponentPose=@n15, bAlphaBoolEnabled:bool=@n16, Translation="2,-2,0", Rotation="0,25,0", Scale="1,1,1") -> (Pose) {
    alphaInputType=Bool
    rotationMode=BMM_Additive
    rotationSpace=BCS_BoneSpace
    translationMode=BMM_Additive
    translationSpace=BCS_BoneSpace
    alphaBoolBlend {blendInTime=0.5, blendOption=Linear, blendOutTime=0.3}
    alphaScaleBiasClamp {bInterpResult=true, interpSpeedIncreasing=2}
    boneToModify {boneName=dyn_hair_bk_MSTR}
    rotation {yaw=25}
    translation=<2 -2>
   }
   n18 = "Transform (Modify) Bone - Bone: dyn_hair_r_MSTR"!(ComponentPose=@n17, bAlphaBoolEnabled:bool=@n16, Translation="2,-2,0", Rotation="0,25,0", Scale="1,1,1") -> (Pose) {
    alphaInputType=Bool
    rotationMode=BMM_Additive
    rotationSpace=BCS_BoneSpace
    translationMode=BMM_Additive
    translationSpace=BCS_BoneSpace
    alphaBoolBlend {blendInTime=0.5, blendOption=Linear, blendOutTime=0.3}
    alphaScaleBiasClamp {bInterpResult=true, interpSpeedIncreasing=2}
    boneToModify {boneName=dyn_hair_r_MSTR}
    rotation {yaw=25}
    translation=<2 -2>
   }
   n19 = "Transform (Modify) Bone - Bone: dyn_hair_l_MSTR"!(ComponentPose=@n18, bAlphaBoolEnabled:bool=@n16, Translation="2,-2,0", Rotation="0,25,0", Scale="1,1,1") -> (Pose) {
    alphaInputType=Bool
    rotationMode=BMM_Additive
    rotationSpace=BCS_BoneSpace
    translationMode=BMM_Additive
    translationSpace=BCS_BoneSpace
    alphaBoolBlend {blendInTime=0.5, blendOption=Linear, blendOutTime=0.3}
    alphaScaleBiasClamp {bInterpResult=true, interpSpeedIncreasing=2}
    boneToModify {boneName=dyn_hair_l_MSTR}
    rotation {yaw=25}
    translation=<2 -2>
   }
   n20 = "Transform (Modify) Bone - Bone: dyn_hair_side_r_MSTR"!(ComponentPose=@n19, bAlphaBoolEnabled:bool=@n16, Translation="0,-5,0", Rotation="0,20,0", Scale="1,1,1") -> (Pose) {
    alphaInputType=Bool
    rotationMode=BMM_Additive
    rotationSpace=BCS_BoneSpace
    translationMode=BMM_Additive
    translationSpace=BCS_BoneSpace
    alphaBoolBlend {blendInTime=0.5, blendOption=Linear, blendOutTime=0.3}
    boneToModify {boneName=dyn_hair_side_r_MSTR}
    rotation {yaw=20}
    translation {y=-5}
   }
   n21 = "Transform (Modify) Bone - Bone: dyn_hair_side_l_MSTR"!(ComponentPose=@n20, bAlphaBoolEnabled:bool=@n16, Translation="0,-5,0", Rotation="0,20,0", Scale="1,1,1") -> (Pose) {
    alphaInputType=Bool
    rotationMode=BMM_Additive
    rotationSpace=BCS_BoneSpace
    translationMode=BMM_Additive
    translationSpace=BCS_BoneSpace
    alphaBoolBlend {blendInTime=0.5, blendOption=Linear, blendOutTime=0.3}
    boneToModify {boneName=dyn_hair_side_l_MSTR}
    rotation {yaw=20}
    translation {y=-5}
   }
   n22 = Get bIsParachuteOpened()
   n23 = "Transform (Modify) Bone - Bone: dyn_hair_bk_MSTR"(ComponentPose=@n21, bAlphaBoolEnabled=@n22, Translation="-2,0,0", Rotation="0,0,0", Scale="1,1,1") {
    alphaInputType=Bool
    rotationMode=BMM_Additive
    rotationSpace=BCS_BoneSpace
    translationMode=BMM_Additive
    translationSpace=BCS_BoneSpace
    alphaBoolBlend {blendInTime=0.1, blendOption=Linear, blendOutTime=0.3}
    alphaScaleBiasClamp {bInterpResult=true, interpSpeedIncreasing=2}
    boneToModify {boneName=dyn_hair_bk_MSTR}
    translation {x=-2}
   }
   n24 = "Transform (Modify) Bone - Bone: dyn_hair_r_MSTR"(ComponentPose=@n23, bAlphaBoolEnabled=@n22, Translation="-2,0,-2", Rotation="0,0,0", Scale="1,1,1") {
    alphaInputType=Bool
    rotationMode=BMM_Additive
    rotationSpace=BCS_BoneSpace
    translationMode=BMM_Additive
    translationSpace=BCS_BoneSpace
    alphaBoolBlend {blendInTime=0.1, blendOption=Linear, blendOutTime=0.3}
    alphaScaleBiasClamp {bInterpResult=true, interpSpeedIncreasing=2}
    boneToModify {boneName=dyn_hair_r_MSTR}
    translation {x=-2, z=-2}
   }
   n25 = "Transform (Modify) Bone - Bone: dyn_hair_l_MSTR"(ComponentPose=@n24, bAlphaBoolEnabled=@n22, Translation="-2,0,2", Rotation="0,0,0", Scale="1,1,1") {
    alphaInputType=Bool
    rotationMode=BMM_Additive
    rotationSpace=BCS_BoneSpace
    translationMode=BMM_Additive
    translationSpace=BCS_BoneSpace
    alphaBoolBlend {blendInTime=0.1, blendOption=Linear, blendOutTime=0.3}
    alphaScaleBiasClamp {bInterpResult=true, interpSpeedIncreasing=2}
    boneToModify {boneName=dyn_hair_l_MSTR}
    translation {x=-2, z=2}
   }
   n26 = "Transform (Modify) Bone - Bone: dyn_hair_side_r_MSTR"(ComponentPose=@n25, bAlphaBoolEnabled=@n22, Translation="-2,0,1", Rotation="0,0,0", Scale="1,1,1") {
    alphaInputType=Bool
    translationMode=BMM_Additive
    translationSpace=BCS_BoneSpace
    alphaBoolBlend {blendInTime=0.1, blendOption=Linear, blendOutTime=0.3}
    boneToModify {boneName=dyn_hair_side_r_MSTR}
    translation {x=-2, z=1}
   }
   n27 = "Transform (Modify) Bone - Bone: dyn_hair_side_l_MSTR"(ComponentPose=@n26, bAlphaBoolEnabled=@n22, Translation="-3,0,3", Rotation="0,0,0", Scale="1,1,1") {
    alphaInputType=Bool
    translationMode=BMM_Additive
    translationSpace=BCS_BoneSpace
    alphaBoolBlend {blendInTime=0.1, blendOption=Linear, blendOutTime=0.3}
    boneToModify {boneName=dyn_hair_side_l_MSTR}
    translation {x=-3, z=3}
   }
   n28 = "Constraint - dyn_hair_bk_01 (PP)"!(ComponentPose=@n27, bAlphaBoolEnabled:bool=@n1, ConstraintWeights_0=0.2, ConstraintWeights_1:float=0.8) -> (Pose) {
    alphaInputType=Bool
    alphaBoolBlend {blendInTime=0.1, blendOption=Linear, blendOutTime=0.1}
    boneToModify {boneName=dyn_hair_bk_01}
    constraintSetup [{targetBone {boneName=head}}, {targetBone {boneName=spine_05}}]
    constraintWeights [1,1]
   }
   n29 = "Constraint - dyn_hair_bk_02 (PP)"!(ComponentPose=@n28, bAlphaBoolEnabled:bool=@n1, ConstraintWeights_0=0.7, ConstraintWeights_1:float=0.3) -> (Pose) {
    alphaInputType=Bool
    alphaBoolBlend {blendInTime=0.1, blendOption=Linear, blendOutTime=0.1}
    boneToModify {boneName=dyn_hair_bk_02}
    constraintSetup [{targetBone {boneName=spine_05}}, {targetBone {boneName=spine_04}}]
    constraintWeights [1,1]
   }
   n30 = "Constraint - dyn_hair_l_01 (PP)"!(ComponentPose=@n29, bAlphaBoolEnabled:bool=@n1, ConstraintWeights_0=0.8, ConstraintWeights_1:float=0.2) -> (Pose) {
    alphaInputType=Bool
    alphaBoolBlend {blendInTime=0.1, blendOption=Linear, blendOutTime=0.1}
    boneToModify {boneName=dyn_hair_l_01}
    constraintSetup [{targetBone {boneName=spine_05}}, {targetBone {boneName=head}}]
    constraintWeights [1,1]
   }
   n31 = "Constraint - dyn_hair_l_02 (PP)"!(ComponentPose=@n30, bAlphaBoolEnabled:bool=@n1, ConstraintWeights_0=0.3, ConstraintWeights_1:float=0.7) -> (Pose) {
    alphaInputType=Bool
    alphaBoolBlend {blendInTime=0.1, blendOption=Linear, blendOutTime=0.1}
    boneToModify {boneName=dyn_hair_l_02}
    constraintSetup [{targetBone {boneName=spine_04}}, {targetBone {boneName=spine_05}}]
    constraintWeights [1,1]
   }
   n32 = "Constraint - dyn_hair_r_01 (PP)"!(ComponentPose=@n31, bAlphaBoolEnabled:bool=@n1, ConstraintWeights_0=0.8, ConstraintWeights_1:float=0.2) -> (Pose) {
    alphaInputType=Bool
    alphaBoolBlend {blendInTime=0.1, blendOption=Linear, blendOutTime=0.1}
    boneToModify {boneName=dyn_hair_r_01}
    constraintSetup [{targetBone {boneName=spine_05}}, {targetBone {boneName=head}}]
    constraintWeights [1,1]
   }
   n33 = "Constraint - dyn_hair_r_02 (PP)"!(ComponentPose=@n32, bAlphaBoolEnabled:bool=@n1, ConstraintWeights_0=0.3, ConstraintWeights_1:float=0.7) -> (Pose) {
    alphaInputType=Bool
    alphaBoolBlend {blendInTime=0.1, blendOption=Linear, blendOutTime=0.1}
    boneToModify {boneName=dyn_hair_r_02}
    constraintSetup [{targetBone {boneName=spine_04}}, {targetBone {boneName=spine_05}}]
    constraintWeights [1,1]
   }
   n34 = "Constraint - dyn_hair_side_l_02 (PP)"!(ComponentPose=@n33, bAlphaBoolEnabled:bool=@n1, ConstraintWeights_0=0.2, ConstraintWeights_1:float=0.8) -> (Pose) {
    alphaInputType=Bool
    alphaBoolBlend {blendInTime=0.1, blendOption=Linear, blendOutTime=0.1}
    boneToModify {boneName=dyn_hair_side_l_02}
    constraintSetup [{targetBone {boneName=spine_05}}, {targetBone {boneName=clavicle_l}}]
    constraintWeights [1,1]
   }
   n35 = "Constraint - dyn_hair_side_l_03 (PP)"!(ComponentPose=@n34, bAlphaBoolEnabled:bool=@n1, ConstraintWeights_0=0.3, ConstraintWeights_1:float=0.7) -> (Pose) {
    alphaInputType=Bool
    alphaBoolBlend {blendInTime=0.1, blendOption=Linear, blendOutTime=0.1}
    boneToModify {boneName=dyn_hair_side_l_03}
    constraintSetup [{targetBone {boneName=spine_05}}, {targetBone {boneName=clavicle_l}}]
    constraintWeights [1,1]
   }
   n36 = "Constraint - dyn_hair_side_l_04 (PP)"!(ComponentPose=@n35, bAlphaBoolEnabled:bool=@n1, ConstraintWeights_0=0.9, ConstraintWeights_1:float=0.1) -> (Pose) {
    alphaInputType=Bool
    alphaBoolBlend {blendInTime=0.1, blendOption=Linear, blendOutTime=0.1}
    boneToModify {boneName=dyn_hair_side_l_04}
    constraintSetup [{targetBone {boneName=spine_04}}, {targetBone {boneName=spine_05}}]
    constraintWeights [1,1]
   }
   n37 = "Constraint - dyn_hair_side_r_02 (PP)"!(ComponentPose=@n36, bAlphaBoolEnabled:bool=@n1, ConstraintWeights_0=0.6, ConstraintWeights_1:float=0.4) -> (Pose) {
    alphaInputType=Bool
    alphaBoolBlend {blendInTime=0.1, blendOption=Linear, blendOutTime=0.1}
    boneToModify {boneName=dyn_hair_side_r_02}
    constraintSetup [{targetBone {boneName=spine_05}}, {targetBone {boneName=clavicle_r}}]
    constraintWeights [1,1]
   }
   n38 = "Constraint - dyn_hair_side_r_03 (PP)"!(ComponentPose=@n37, bAlphaBoolEnabled:bool=@n1, ConstraintWeights_0=0.3, ConstraintWeights_1:float=0.7) -> (Pose) {
    alphaInputType=Bool
    alphaBoolBlend {blendInTime=0.1, blendOption=Linear, blendOutTime=0.1}
    boneToModify {boneName=dyn_hair_side_r_03}
    constraintSetup [{targetBone {boneName=spine_05}}, {targetBone {boneName=clavicle_r}}]
    constraintWeights [1,1]
   }
   n39 = "Constraint - dyn_hair_side_r_04 (PP)"!(ComponentPose=@n38, bAlphaBoolEnabled:bool=@n1, ConstraintWeights_0=0.9, ConstraintWeights_1:float=0.1) -> (Pose) {
    alphaInputType=Bool
    alphaBoolBlend {blendInTime=0.1, blendOption=Linear, blendOutTime=0.1}
    boneToModify {boneName=dyn_hair_side_r_04}
    constraintSetup [{targetBone {boneName=spine_04}}, {targetBone {boneName=spine_05}}]
    constraintWeights [1,1]
   }
   n40 = "Constraint - dyn_hair_r_Front_00 (PP)"!(ComponentPose=@n39, bAlphaBoolEnabled:bool=@n1, ConstraintWeights_0=0.7, ConstraintWeights_1:float=0.3) -> (Pose) {
    alphaInputType=Bool
    alphaBoolBlend {blendInTime=0.1, blendOption=Linear, blendOutTime=0.1}
    boneToModify {boneName=dyn_hair_r_Front_00}
    constraintSetup [{targetBone {boneName=spine_05}}, {targetBone {boneName=head}}]
    constraintWeights [1,1]
   }
   n41 = "Constraint - dyn_hair_r_Front_01 (PP)"!(ComponentPose=@n40, bAlphaBoolEnabled:bool=@n1, ConstraintWeights_0=0.8, ConstraintWeights_1:float=0.2) -> (Pose) {
    alphaInputType=Bool
    alphaBoolBlend {blendInTime=0.1, blendOption=Linear, blendOutTime=0.1}
    boneToModify {boneName=dyn_hair_r_Front_01}
    constraintSetup [{targetBone {boneName=spine_05}}, {targetBone {boneName=head}}]
    constraintWeights [1,1]
   }
   n42 = "Constraint - dyn_hair_r_Front_02 (PP)"!(ComponentPose=@n41, bAlphaBoolEnabled:bool=@n1, ConstraintWeights_0=1, ConstraintWeights_1:float=0) -> (Pose) {
    alphaInputType=Bool
    alphaBoolBlend {blendInTime=0.1, blendOption=Linear, blendOutTime=0.1}
    boneToModify {boneName=dyn_hair_r_Front_02}
    constraintSetup [{targetBone {boneName=spine_05}}, {targetBone {boneName=head}}]
    constraintWeights [1,1]
   }
   n43 = "Constraint - dyn_hair_r_Front_03 (PP)"!(ComponentPose=@n42, bAlphaBoolEnabled:bool=@n1, ConstraintWeights_0=1, ConstraintWeights_1:float=0) -> (Pose) {
    alphaInputType=Bool
    alphaBoolBlend {blendInTime=0.1, blendOption=Linear, blendOutTime=0.1}
    boneToModify {boneName=dyn_hair_r_Front_03}
    constraintSetup [{targetBone {boneName=spine_05}}, {targetBone {boneName=head}}]
    constraintWeights [1,1]
   }
   n44 = RigidBody!(
    ComponentPose=@n43
    Alpha
    OverrideWorldGravity:Vector=Get GravityOverride()
    ExternalForce
    ComponentLinearAccScale:Vector=Get ACC() -> (ACC:Vector)
    ComponentLinearVelScale:Vector=Get VEL() -> (VEL:Vector)
   ) -> (Pose) {
    bClampLinearTranslationLimitToRefPose=true
    bOverrideWorldGravity=true
    overridePhysicsAsset=&/F_MED_FishBowl_FaceAcc_Physics
    simulationSpace=BaseBoneSpace
    alphaBoolBlend {blendOption=Linear}
    baseBoneRef {boneName=spine_05}
    componentAppliedLinearAccClamp=<1000 2500 2500>
    componentLinearAccScale=<0.3 0.3 0.5>
    componentLinearVelScale=<0.3 0.3 0.5>
    overrideWorldGravity {z=-800}
    simSpaceSettings {
     maxAngularAcceleration=1000
     maxAngularVelocity=3
     maxLinearAcceleration=700
     maxLinearVelocity=500
     velocityScaleZ=2
     worldAlpha=0.5
     externalLinearDragV=<0.2 0.2>
    }
   }
   n45 = RigidBody!(ComponentPose=@n44, Alpha, OverrideWorldGravity="0,0,-980", ExternalForce) -> (Pose) {
    bClampLinearTranslationLimitToRefPose=true
    bOverrideWorldGravity=true
    overridePhysicsAsset=&/F_MED_FishBowl_FaceAcc_Fr_Physics
    simulationSpace=BaseBoneSpace
    alphaBoolBlend {blendOption=Linear}
    baseBoneRef {boneName=spine_05}
    componentAppliedLinearAccClamp=<500 500 500>
    componentLinearAccScale=<0.2 0.2 0.2>
    componentLinearVelScale=<0.2 0.2 0.2>
   }
   n46 = Anim Dynamics - Bone: dyn_hair_front_r_01(ComponentPose=@n45, Alpha=1, ExternalForce="0,0,0") {
    angularDampingOverride=0.98
    angularSpringConstant=50
    bAngularSpring=true
    bGravityOverrideInSimSpace=true
    bOverrideAngularDamping=true
    simulationSpace=RootRelative
    alphaBoolBlend {blendOption=Linear}
    boundBone {boneName=dyn_hair_front_r_01}
    componentAppliedLinearAccClamp=<3500 3500 3500>
    componentLinearAccScale=<0.6 0.6 1.2>
    componentLinearVelScale=<0.6 0.6 1.2>
    physicsBodyDefinitions [{
      sphereCollisionRadius=0
      boundBone {boneName=dyn_hair_front_r_01}
      boxExtents=<20 15 15>
      constraintSetup {angularLimitsMax=<1 7 2>, angularLimitsMin=<-1 -15 -15>, angularTarget=<1 0 0>}
      localJointOffset=<2 0 0>
     }]
    relativeSpaceBone {boneName=root}
   }
   n47 = Anim Dynamics - Bone: dyn_hair_front_r_02(ComponentPose=@n46, Alpha=1, ExternalForce="0,0,0") {
    angularDampingOverride=0.98
    angularSpringConstant=50
    bAngularSpring=true
    bGravityOverrideInSimSpace=true
    bOverrideAngularDamping=true
    simulationSpace=RootRelative
    alphaBoolBlend {blendOption=Linear}
    boundBone {boneName=dyn_hair_front_r_02}
    componentAppliedLinearAccClamp=<3500 3500 3500>
    componentLinearAccScale=<0.6 0.6 1.2>
    componentLinearVelScale=<0.6 0.6 1.2>
    physicsBodyDefinitions [{
      sphereCollisionRadius=0
      boundBone {boneName=dyn_hair_front_r_02}
      boxExtents=<20 15 15>
      constraintSetup {angularLimitsMax=<1 7 2>, angularLimitsMin=<-1 -15 -15>, angularTarget=<1 0 0>}
      localJointOffset=<2 0 0>
     }]
    relativeSpaceBone {boneName=root}
   }
   Output Pose(Result=Component To Local(ComponentPose=@n47)) {layerGroup=None}
  ]}
 EventGraph {nodes [
   Event BlueprintUpdateAnimation()
   > n1 = Sequence!(execute) -> (then_0, then_1, then_2:Exec, then_3:Exec)
   > n2 = Gravity Override State Function DEPRECATED(
    CharacterPartAnimInstance=Self-Reference()
    GameTimeInSeconds=GetGameTimeInSeconds()
    GravityOverrideParams=Get GravityOverrideParams() -> (GravityOverrideParams:Gravity Override Params Struct Structure)
   )
   > Set GravityOverride(GravityOverride=@n2)
   > n3 = Branch(Condition=CheckForBackbling())
   > Set ACC(ACC:Vector="0, 0, 0") -> (Output_Get:Vector)
   > Set VEL(VEL:Vector="0, 0, 0") -> (Output_Get:Vector)
   > Set GravityOverride(GravityOverride="0,0,-980")
   > Set backpack(backpack:double=0.5) -> (Output_Get:double)
   > Set jumping(jumping:double=0) -> (Output_Get:double)
   Set ACC(execute=@n3.else, ACC="0.2,0.3,0.3")
   > Set VEL(VEL="0.2,0.3,0.3")
   > Set backpack(backpack=0)
   Branch(execute=@n1.then_1, Condition=Get bIsSkyDiving())
   > Set jumping(jumping=0)
   n4 = Get bIsCrouching()
   n5 = Branch(execute=@n1.then_2, Condition=AND Boolean(A=@n4, B=Get bIsMoving2D()))
   > Set crouchmove(crouchmove:double=1) -> (Output_Get:double)
   Set crouchmove(execute=@n5.else, crouchmove=0)
   Set crouch(execute=@n1.then_3, crouch:double="To Float (Boolean)"(InBool=@n4)) -> (Output_Get:double)
   n6 = Event OnBeginFalling()
   Set jumping(execute=[@n6.then, @n8.then], jumping=1)
   n7 = Event OnEndFalling()
   Set jumping(execute=[@n7.then, @n9.then], jumping=0)
   n8 = Event OnBeginJumping()
   n9 = Event OnEndJumping()
  ]}
]
variables [
 ACC {type=Vector}
 backpack {type=number}
 crouch {type=number}
 crouchmove {type=number}
 GravityOverride {type=Vector}
 GravityOverrideParams {
  type=GravityOverrideParamsStruct
  default {
   onGround_Falling {gravityOverride=<0 0 -1250>}
   onGround_Moving {gravityOverride=<0 0 -1150>}
   skydive_FloatingDown {gravityOverride=<0 -100 50>, windAmplitude=<75 200 150>, windFrequency=<1 3.5 6>, windOffset=<0 -200 125>}
   skydive_GoingUp {gravityOverride=<0 -300 -600>}
   tactical_Sprint {gravityOverride=<0 0 -1150>}
  }
 }
 jumping {type=number}
 VEL {type=Vector}
]
